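; Rift portal actor variant for the Figment Cube that adds a timed forced close and a proximity-driven post-process effect.
; Two seconds after BeginPlay, an open portal deactivates its capsule, calls PortalOpened and schedules DestroyRift after its active duration.
; Radius overlaps toggle the rift post process based on whether any player pawns remain inside, only where cosmetic events can run.
; Parent destruction logic runs only when DestroyRiftAfterUse is set.
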(event EventBeginPlay
  ;[1 1 1] EVENT--> BeginPlay
  (|CallParentFunction)
  (Utilities|Time|SetTimerbyEvent (EventDispatchers|CreateEvent :!event Custom|ForceClosePortal) 2.0)
  ;[1 1 1] Deactivate the Rift Post Process on Begin Play
  (bind _riftpostprocess (Variables|Default|GetRiftPostProcess)
    (:then
      (Components|Activation|Deactivate _riftpostprocess))
    (:else)))

(event EventDestroyRift_Internal
  ;[1 1 1] On Death
  (if (Variables|Default|GetDestroyRiftAfterUse)
    (|CallParentFunction)))

(event OnComponentBeginOverlap(PostProcessRadius)
  (params (OverlappedComponent :type Primitive%20Component%20Object%20Reference) (OtherActor :type Actor%20Object%20Reference) (OtherComp :type Primitive%20Component%20Object%20Reference) (OtherBodyIndex :type Integer) (bFromSweep :type Boolean) (SweepResult :type Hit%20Result%20Structure%20%28by%20ref%29))
  ;[1 1 1] Handle Rift Post Process to Activate and Enable on Player Begin Overlap and Deactivate and Disable on Player End Overlap
  (Networking|CanExecuteCosmeticEvents
    (:True
      ;;[1 1 1] Only Activate and Enable if there are 1 or more pawns in the volume and the Rift Post Process is not active
      (Utilities|Casting|CastToFortPlayerPawn OtherActor)
      (bind _riftpostprocess (Variables|Default|GetRiftPostProcess)
        (:then
          (if (and (not (Components|Activation|IsActive _riftpostprocess)) (> (Utilities|Array|Length (Collision|GetOverlappingActors (Variables|Default|GetPostProcessRadius) "/Script/FortniteGame.FortPlayerPawn" :!class /Script/Engine.PrimitiveComponent)) 0))
            (Components|Activation|Activate _riftpostprocess false)
            (Variables|/Script/Engine.PostProcessComponent|SetbEnabled true _riftpostprocess)))
        (:else)))))

(event OnComponentEndOverlap(PostProcessRadius)
  (params (OverlappedComponent :type Primitive%20Component%20Object%20Reference) (OtherActor :type Actor%20Object%20Reference) (OtherComp :type Primitive%20Component%20Object%20Reference) (OtherBodyIndex :type Integer))
  ;[1 1 1] Handle Rift Post Process to Activate and Enable on Player Begin Overlap and Deactivate and Disable on Player End Overlap
  (Networking|CanExecuteCosmeticEvents
    (:True
      ;;[1 1 1] Only Deactivate and Disable the Post Process volume if there are no other players in the volume and the Post Process is active for the Client
      (Utilities|Casting|CastToFortPlayerPawn OtherActor)
      (bind _riftpostprocess (Variables|Default|GetRiftPostProcess)
        (:then
          (if (and (Components|Activation|IsActive _riftpostprocess) (<= (Utilities|Array|Length (Collision|GetOverlappingActors (Variables|Default|GetPostProcessRadius) "/Script/FortniteGame.FortPlayerPawn" :!class /Script/Engine.PrimitiveComponent)) 0))
            (Components|Activation|Deactivate _riftpostprocess)
            (Variables|/Script/Engine.PostProcessComponent|SetbEnabled false _riftpostprocess)))
        (:else)))))

(event Custom|ForceClosePortal
  (if (Variables|RiftPortal|GetPortalOpen)
    (bind _capsule (Variables|Default|GetCapsule)
      (:then
        (Components|Activation|Deactivate _capsule)
        (Utilities|Time|SetTimerbyEvent (EventDispatchers|CreateEvent :!event DestroyRift) (ScalableFloat|GetValueatLevel (Variables|RiftPortal|GetActiveDuration)))
        (CallFunction|PortalOpened))
      (:else))))

(fn ConstructionScript)
